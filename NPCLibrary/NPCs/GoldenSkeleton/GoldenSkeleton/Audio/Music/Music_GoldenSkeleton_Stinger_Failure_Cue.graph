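# Sound cue for the Golden Skeleton NPC's failure music stinger.
# A wave player feeds the fail stinger wave through a delay node, which adds a short fixed pause before playback.
# The cue routes to the NPC music sound class with slightly boosted volume and a very large max distance.
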SoundCue /NPCLibrary/NPCs/GoldenSkeleton/GoldenSkeleton/Audio/Music/Music_GoldenSkeleton_Stinger_Failure_Cue
root=Delay_1
nodes [
 {id=Delay_1, children [WavePlayer_2], properties {delayMax=0.1, delayMin=0.1}}
 {id=WavePlayer_2, properties {soundWaveAssetPtr=&/GoldenSkeleton_Music_Stinger_Fail}}
]
properties {maxDistance=2097152, soundClassObject=&Fort_Music_NPC, volumeMultiplier=1.2}
