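# Sound cue for the Golden Skeleton NPC's success music stinger.
# A single wave player plays the success stinger wave through a fixed short delay at the root.
# It routes to the NPC music sound class.
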SoundCue /NPCLibrary/NPCs/GoldenSkeleton/GoldenSkeleton/Audio/Music/Music_GoldenSkeleton_Stinger_Success_Cue
root=Delay_0
nodes [
 {id=Delay_0, children [WavePlayer_1], properties {delayMax=0.2, delayMin=0.2}}
 {
  id=WavePlayer_1
  properties {soundWaveAssetPtr=&/GoldenSkeleton_Music_Stinger_Success}
 }
]
properties {maxDistance=2097152, soundClassObject=&Fort_Music_NPC, volumeMultiplier=1}
